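# Physics asset for the PER mesh part of the F_MED_Pliant female player cosmetic, covering hair dynamics and upper-body collision shapes.
# Three hair chains (A, B, C) of small capsules branch from a hair root on the head, with tight swing and twist limits along each chain, stiffest on C.
# Capsules and boxes cover upper spine, neck, head and left collar, linked by angularly free constraints, as are the hair root and chain bases.
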PhysicsAsset /BRCosmetics/Characters/Player/Female/Medium/Bodies/F_MED_Pliant/Meshes/Parts/F_MED_Pliant_PER/F_MED_Pliant_PER_Physics
bodies {
 asn1_hair [{shapeType=Capsule, length=1, radius=1}]
 asn1_hair_main_A_a [{shapeType=Capsule, length=3, radius=1.75, center=<3.105 0 0>, rotation=<rot 90 0 0>}]
 asn1_hair_main_A_a_base [{shapeType=Capsule, length=1, radius=1}]
 asn1_hair_main_A_b [{shapeType=Capsule, length=2.25, radius=2, center=<2.7376 0 0>, rotation=<rot 86.3848 0 0>}]
 asn1_hair_main_A_c [{shapeType=Capsule, length=1.3799, radius=1.5, center=<2.022 0 0>, rotation=<rot 90 0 0>}]
 asn1_hair_main_B_a [{shapeType=Capsule, length=2.5, radius=2.5, center=<3.0618 0 0>, rotation=<rot 90 0 0>}]
 asn1_hair_main_B_a_base [{shapeType=Capsule, length=1, radius=1}]
 asn1_hair_main_B_b [{shapeType=Capsule, length=1.75, radius=2.25, center=<2.5268 0 0>, rotation=<rot 90 0 0>}]
 asn1_hair_main_C_a [{shapeType=Capsule, length=2.5, radius=2.25, center=<3.4616 0 0.0001>, rotation=<rot 90 0 0>}]
 asn1_hair_main_C_a_base [{shapeType=Capsule, length=1, radius=1}]
 asn1_hair_main_C_b [{shapeType=Capsule, length=1.75, radius=1.75, center=<2.2691 0.0001 0>, rotation=<rot 90 0 0>}]
 dyn_collar_l_MSTR [{
   shapeType=Box
   extentX=19.3316
   extentY=16.9625
   extentZ=7.2353
   shapeName=dyn_collar_l_MSTR_box
   center=<4.12 0.1358 -4.4218>
   rotation=<rot 6.2285 -88.6137 4.5987>
  }]
 head [{shapeType=Capsule, length=5, radius=6.5, center=<2.4796 2.289 0.6971>, rotation=<rot 90 0 0>}]
 neck_01 [{shapeType=Capsule, length=4.75, radius=4.5, center=<2.166 0.4802 0>, rotation=<rot 90 0 0.8586>}]
 neck_02 [
  {
   shapeType=Box
   extentX=14.6971
   extentY=22.9547
   extentZ=8.8034
   shapeName=neck_02_box
   center=<-4.6405 -8.9597 0.1302>
   rotation=<rot 0.588 -30.1137 -89.2164>
  }
  {
   shapeType=Box
   extentX=17.605
   extentY=7.5
   extentZ=25.6754
   shapeName=neck_02_box
   center=<-3.8225 -3.8931 -10.7033>
   rotation=<rot -12.5312 -17.7815 -56.1829>
  }
 ]
 spine_05 [{shapeType=Capsule, length=10, radius=7.75, center=<1.9244 -0.67 0>}]
}
constraints [
 {
  bone1Name=head
  bone2Name=neck_02
  swing1LimitDegrees=45
  swing1Motion=Free
  swing2LimitDegrees=45
  swing2Motion=Free
  twistLimitDegrees=45
  twistMotion=Free
 }
 {
  bone1Name=neck_01
  bone2Name=spine_05
  swing1LimitDegrees=45
  swing1Motion=Free
  swing2LimitDegrees=45
  swing2Motion=Free
  twistLimitDegrees=45
  twistMotion=Free
 }
 {
  bone1Name=asn1_hair
  bone2Name=head
  swing1LimitDegrees=45
  swing1Motion=Free
  swing2LimitDegrees=45
  swing2Motion=Free
  twistLimitDegrees=45
  twistMotion=Free
 }
 {
  bone1Name=asn1_hair_main_A_a_base
  bone2Name=asn1_hair
  swing1LimitDegrees=45
  swing1Motion=Free
  swing2LimitDegrees=45
  swing2Motion=Free
  twistLimitDegrees=45
  twistMotion=Free
 }
 {
  bone1Name=asn1_hair_main_B_a_base
  bone2Name=asn1_hair
  swing1LimitDegrees=45
  swing1Motion=Free
  swing2LimitDegrees=45
  swing2Motion=Free
  twistLimitDegrees=45
  twistMotion=Free
 }
 {
  bone1Name=asn1_hair_main_C_a_base
  bone2Name=asn1_hair
  swing1LimitDegrees=45
  swing1Motion=Free
  swing2LimitDegrees=45
  swing2Motion=Free
  twistLimitDegrees=45
  twistMotion=Free
 }
 {bone1Name=asn1_hair_main_A_a, bone2Name=asn1_hair_main_A_a_base, swing1LimitDegrees=15, swing2LimitDegrees=15, twistLimitDegrees=5}
 {bone1Name=asn1_hair_main_A_b, bone2Name=asn1_hair_main_A_a, swing1LimitDegrees=15, swing2LimitDegrees=15, twistLimitDegrees=5}
 {bone1Name=asn1_hair_main_A_c, bone2Name=asn1_hair_main_A_b, swing1LimitDegrees=15, swing2LimitDegrees=15, twistLimitDegrees=5}
 {bone1Name=asn1_hair_main_B_a, bone2Name=asn1_hair_main_B_a_base, swing1LimitDegrees=15, swing2LimitDegrees=15, twistLimitDegrees=5}
 {bone1Name=asn1_hair_main_B_b, bone2Name=asn1_hair_main_B_a, swing1LimitDegrees=15, swing2LimitDegrees=15, twistLimitDegrees=5}
 {bone1Name=asn1_hair_main_C_a, bone2Name=asn1_hair_main_C_a_base, swing1LimitDegrees=10, swing2LimitDegrees=10, twistLimitDegrees=3}
 {bone1Name=asn1_hair_main_C_b, bone2Name=asn1_hair_main_C_a, swing1LimitDegrees=10, swing2LimitDegrees=10, twistLimitDegrees=3}
 {
  bone1Name=dyn_collar_l_MSTR
  bone2Name=spine_05
  swing1LimitDegrees=45
  swing1Motion=Free
  swing2LimitDegrees=45
  swing2Motion=Free
  twistLimitDegrees=45
  twistMotion=Free
 }
 {
  bone1Name=neck_02
  bone2Name=neck_01
  swing1LimitDegrees=45
  swing1Motion=Free
  swing2LimitDegrees=45
  swing2Motion=Free
  twistLimitDegrees=45
  twistMotion=Free
 }
]
